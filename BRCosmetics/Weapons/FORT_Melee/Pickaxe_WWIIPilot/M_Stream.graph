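# Translucent two-sided mesh-particle material for a smoky stream effect on the WWII Pilot pickaxe.
# Two panning smoke samples are screen-blended, masked by particle alpha and vertex red (plus depth fade, except on mobile), then contrast-adjusted.
# Time-of-day dark and highlight colors blend by smoke alpha, tinted by particle color.
# Time-scrolled noise forms a normalized wobble vector that grows along the V coordinate.
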Material /BRCosmetics/Weapons/FORT_Melee/Pickaxe_WWIIPilot/M_Stream
graph {
 properties {bUsedWithMeshParticles=true, blendMode=BLEND_Translucent, pixelDepthOffsetMode=PDOM_Legacy, twoSided=true, usedShadingModels=Default Lit}
 nodes [
  n1 = Multiply(A=ParticleColor().A, B=VertexColor().R)
  n2 = Multiply(A=@n1, B=DepthFade() {fadeDistanceDefault=5})
  n3 = TextureSample(UVs=Panner(Coordinate=TextureCoordinate() {uTiling=0.15, vTiling=0.2}) {speedX=-0.1, speedY=-0.6}) {texture=/Game/Global/Textures/Default/Tileable_Patterns/T_Smoke_OUT}
  n4 = TextureSample(UVs=Panner(Coordinate=TextureCoordinate() {uTiling=0.25, vTiling=0.25}) {speedX=0.07, speedY=-0.5}) {texture=&T_Smoke_OUT}
  MaterialFunctionCall(
   In=Multiply(
    A=ShadingPathSwitch(Default=@n2, Mobile=@n1) {desc=Migrated from FeatureLevelSwitch}
    B=MaterialFunctionCall!(Blend=@n3.R, Base=@n4.R) {materialFunction=/Engine/Functions/Engine_MaterialFunctions03/Blends/Blend_Screen}.Result
   )
   Contrast=Constant() {r=0.1}
  ) {materialFunction=&CheapContrast}
  n5 = VectorParameter() {group=Smoke Colors, parameterName=4 - Nighttime Darks Smoke, defaultValue=<rgba 0.0428 0.1436 0.2761 1>}
  n6 = MaterialFunctionCall(
   Color_Morning=VectorParameter() {group=Smoke Colors, parameterName=1 - Morning Darks Smoke, defaultValue=<rgba 0.22 0.0781 0.0352 1>}.RGB
   Color_Day=VectorParameter() {group=Smoke Colors, parameterName=2 - Daytime Darks Smoke, defaultValue=<rgba 0.25 0.3716 0.5 1>}.RGB
   Color_Evening=VectorParameter() {group=Smoke Colors, parameterName=3 - Evening Darks Smoke, defaultValue=<rgba 0.5894 0.2269 0.1267 1>}.RGB
   Color_Night=@n5.RGB
  ) {materialFunction=&Color_Time_of_Day}
  n7 = VectorParameter() {group=Smoke Colors, parameterName=1 - Morning Highlights Smoke, defaultValue=<rgba 0.67 0.375 0.1407 1>}
  n8 = VectorParameter() {group=Smoke Colors, parameterName=3 - Evening Highlights Smoke, defaultValue=<rgba 0.56 0.3409 0.2856 1>}
  n9 = VectorParameter() {group=Smoke Colors, parameterName=4 - Nighttime Highlights Smoke, defaultValue=<rgba 0.258 0.2833 0.46 1>}
  n10 = MaterialFunctionCall(
   Color_Morning=@n7.RGB
   Color_Day=VectorParameter() {group=Smoke Colors, parameterName=2 - Daytime Highlights Smoke, defaultValue=<rgba 0.792 0.824 0.9 1>}.RGB
   Color_Evening=@n8.RGB
   Color_Night=@n9.RGB
  ) {materialFunction=&Color_Time_of_Day}
  Multiply(A=Multiply(A=LinearInterpolate(A=@n6.Result, B=@n10.Result, Alpha=@n3.A)) {constB=0.15}, B=ParticleColor().RGB)
  n11 = ComponentMask(None=TextureCoordinate()) {g=true}
  n12 = Subtract(A=Time(), B=Multiply(A=@n11) {constB=0.1})
  n13 = TextureSample(UVs=AppendVector(A=@n12, B=Multiply(A=@n12) {constB=0.1})) {samplerType=SAMPLERTYPE_LinearColor, texture=&HighContrastNoise}
  Multiply(A=Normalize(VectorInput=ConstantBiasScale(None=@n13.RGB) {bias=-0.5, scale=2}), B=Multiply(A=@n11) {constB=15})
 ]
}
