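# Translucent lit beam-trail material for the WWII Pilot pickaxe, rendering a smoke trail with a hot emissive tint.
# Screen-aligned panning puff noise blends dark and highlight smoke colors chosen by time of day.
# That noise multiplies a soft trail texture panned by particle-driven time, shaped by a sine UV falloff and squared alpha parameter.
# The mask drives emissive glow and, with particle alpha and depth fade, a subtle index-of-refraction-style lerp.
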Material /BRCosmetics/Weapons/FORT_Melee/Pickaxe_WWIIPilot/M_WWIIPilot_Trail
graph {
 properties {
  bUsedWithBeamTrails=true
  blendMode=BLEND_Translucent
  numCustomizedUVs=3
  pixelDepthOffsetMode=PDOM_Legacy
  refractionMethod=RM_IndexOfRefraction
  usedShadingModels=Default Lit
 }
 nodes [
  n1 = VectorParameter() {group=Smoke Colors, parameterName=4 - Nighttime Darks Smoke, defaultValue=<rgba 0.0428 0.1436 0.2761 1>}
  n2 = VectorParameter() {group=Smoke Colors, parameterName=1 - Morning Highlights Smoke, defaultValue=<rgba 0.67 0.375 0.1407 1>}
  n3 = VectorParameter() {group=Smoke Colors, parameterName=3 - Evening Highlights Smoke, defaultValue=<rgba 0.56 0.3409 0.2856 1>}
  n4 = VectorParameter() {group=Smoke Colors, parameterName=4 - Nighttime Highlights Smoke, defaultValue=<rgba 0.258 0.2833 0.46 1>}
  n5 = Constant() {r=512}
  n6 = TextureSample(UVs=Panner(Coordinate=MaterialFunctionCall(Texture Dimensions=@n5) {materialFunction=&ScreenAlignedUVs}."X 100%, Y 100%") {speedY=0.5}) {texture=/Game/Global/Textures/Default/Tileable_Patterns/T_Smoke_Puff_Tile_Up}
  n7 = LinearInterpolate(
   A=MaterialFunctionCall(
    Color_Morning=VectorParameter() {group=Smoke Colors, parameterName=1 - Morning Darks Smoke, defaultValue=<rgba 0.22 0.0781 0.0352 1>}.RGB
    Color_Day=VectorParameter() {group=Smoke Colors, parameterName=2 - Daytime Darks Smoke, defaultValue=<rgba 0.25 0.3716 0.5 1>}.RGB
    Color_Evening=VectorParameter() {group=Smoke Colors, parameterName=3 - Evening Darks Smoke, defaultValue=<rgba 0.5894 0.2269 0.1267 1>}.RGB
    Color_Night=@n1.RGB
   ) {materialFunction=&Color_Time_of_Day}.Result
   B=MaterialFunctionCall(
    Color_Morning=@n2.RGB
    Color_Day=VectorParameter() {group=Smoke Colors, parameterName=2 - Daytime Highlights Smoke, defaultValue=<rgba 0.792 0.824 0.9 1>}.RGB
    Color_Evening=@n3.RGB
    Color_Night=@n4.RGB
   ) {materialFunction=&Color_Time_of_Day}.Result
   Alpha=@n6.R
  )
  n8 = Constant3Vector() {constant=<rgba -1 -1 -1 0>}
  n9 = Constant3Vector() {constant=<rgba 2.5 1.0647 0.5729 0>}
  n10 = DynamicParameter() {paramNames [Emissive,EmtTime,EmtAlpha,GlobalAlpha]}
  n11 = Panner(Coordinate=TextureCoordinate() {coordinateIndex=1, uTiling=0.5}, Time=@n10.EmtTime) {speedX=-2, speedY=-0.1}
  n12 = Multiply(A=Multiply(
    A=@n6.R
    B=TextureSample(UVs=@n11) {samplerType=SAMPLERTYPE_Grayscale, texture=&T_Smoke_Trail_Soft}.R
   )) {constB=2}
  n13 = TextureCoordinate() {uTiling=0.5, vTiling=0.5}
  n14 = Multiply(A=Multiply(A=@n10.EmtAlpha, B=@n10.EmtAlpha), B=ComponentMask(None=Sine(None=@n13)) {g=true})
  n15 = Multiply(
   A=DynamicParameter() {paramNames [Emissive,EmtTime,EmtAlpha,GlobalAlpha]}.Emissive
   B=Multiply(A=LinearInterpolate(Alpha=@n12) {constA=0.25, constB=0.7}, B=@n14)
  )
  Add(A=Multiply(A=@n7) {constB=0.075}, B=Clamp(None=LinearInterpolate(A=@n8, B=@n9, Alpha=@n15)) {clampMode=CMODE_ClampMin})
  n16 = Multiply(A=@n12, B=@n14)
  n17 = Constant() {r=0.1}
  n18 = ParticleColor()
  n19 = ShadingPathSwitch(Default=DepthFade() {fadeDistanceDefault=25}) {desc=Migrated from FeatureLevelSwitch}
  LinearInterpolate(Alpha=Multiply(
    A=MaterialFunctionCall(In=@n16, Contrast=@n17) {materialFunction=&CheapContrast}.Result
    B=Multiply(A=@n18.A, B=@n19)
   )) {constA=1, constB=0.985}
  Multiply(A=@n7) {constB=0.15}
  Transform(None=Constant3Vector() {constant=<rgba 0 0 1 0>}) {transformSourceType=TRANSFORMSOURCE_World, transformType=TRANSFORM_Tangent}
 ]
}
