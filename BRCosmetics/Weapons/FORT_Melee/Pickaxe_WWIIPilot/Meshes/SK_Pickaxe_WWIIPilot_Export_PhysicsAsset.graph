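# Physics asset for the WWII Pilot pickaxe skeletal mesh, a two-handed melee cosmetic with a dynamic blades bone.
# It has two box bodies, a long one on the handle and a wide, thin one on dyn_blades.
# A single constraint joins the blades to the handle with free twist and zero-degree swing limits, so the blades can spin about one axis like a propeller.
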PhysicsAsset /BRCosmetics/Weapons/FORT_Melee/Pickaxe_WWIIPilot/Meshes/SK_Pickaxe_WWIIPilot_Export_PhysicsAsset
bodies {
 Melee_TwoHanded_Handle [{shapeType=Box, extentX=11.5793, extentY=12.8551, extentZ=108.1307, center=<1.0422 -1.0069 42.9921>}]
 dyn_blades [{shapeType=Box, extentX=9.2105, extentY=94.0953, extentZ=5.7886, center=<-0.0243 -0.6644 1.7958>}]
}
constraints [{bone1Name=dyn_blades, bone2Name=Melee_TwoHanded_Handle, swing1LimitDegrees=0, swing2LimitDegrees=0, twistLimitDegrees=180, twistMotion=Free}]
