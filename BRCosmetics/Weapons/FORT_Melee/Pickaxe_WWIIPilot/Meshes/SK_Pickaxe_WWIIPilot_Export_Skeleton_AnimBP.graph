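# Animation Blueprint for the WWII Pilot pickaxe skeleton that adds secondary physics motion, including a swinging hose.
# The input pose runs through a RigidBody node using the pickaxe's physics asset in base-bone space, then an Anim Dynamics node on dyn_hose with spring and angular limits.
# The event graph stores the owner's skeletal mesh in MeshToCopy on initialize, but no node here reads it.
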AnimBlueprint /BRCosmetics/Weapons/FORT_Melee/Pickaxe_WWIIPilot/Meshes/SK_Pickaxe_WWIIPilot_Export_Skeleton_AnimBP
target_skeleton=&/SK_Pickaxe_WWIIPilot_Export_Skeleton
graphs [
 AnimGraph {nodes [
   AnimGraphNode_SubInput_0 = Input Pose()
   n1 = RigidBody(ComponentPose=Local To Component(LocalPose=@AnimGraphNode_SubInput_0)) {
    overridePhysicsAsset=&/SK_Pickaxe_WWIIPilot_Export_PhysicsAsset
    simulationSpace=BaseBoneSpace
    alphaBoolBlend {blendOption=Linear}
    componentLinearAccScale=<1 1 1>
   }
   n2 = Anim Dynamics - Bone: dyn_hose(ComponentPose=@n1, Alpha=1, ExternalForce="0,0,0") {
    angularSpringConstant=500
    bGravityOverrideInSimSpace=true
    alphaBoolBlend {blendOption=Linear}
    boundBone {boneName=dyn_hose}
    componentAppliedLinearAccClamp=<10000 10000 10000>
    componentLinearAccScale=<50 50 50>
    physicsBodyDefinitions [{
      sphereCollisionRadius=0
      boundBone {boneName=dyn_hose}
      boxExtents=<20 20 5>
      constraintSetup {angularLimitsMax=<20 5 45>, angularLimitsMin=<-20 -5 -45>, angularTarget=<0 0 0>}
      localJointOffset=<0 0 -10>
     }]
   }
   Output Pose(Result=Component To Local(ComponentPose=@n2)) {layerGroup=None}
  ]}
 EventGraph {nodes [
   Event BlueprintInitializeAnimation()
   > Set MeshToCopy(MeshToCopy:Skeletal Mesh Component Object Reference=GetOwnerSkeletalMeshComponent()) -> (Output_Get:Skeletal Mesh Component Object Reference)
   Event BlueprintUpdateAnimation()
  ]}
]
variables [MeshToCopy {type=SkeletalMeshComponent}]
